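# Sound cue for the looping audio of the Alien Artifact quest object in Athena, routed through the quests sound class.
# A quality-level node picks between two looping waves, using the standard loop for the first two levels and the "opt" variant for the last two.
# It overrides attenuation with a short-range falloff, while also referencing the close supply-drop marker attenuation asset.
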SoundCue /Game/Athena/Sounds/Quests/Athena_AlienArtifact_Loop_Cue
root=QualityLevel_0
nodes [
 {id=QualityLevel_0, children [WavePlayer_2,WavePlayer_2,WavePlayer_3,WavePlayer_3]}
 {id=WavePlayer_2, properties {bLooping=true, soundWaveAssetPtr=&/Alien_Artifact_Loop}}
 {id=WavePlayer_3, properties {bLooping=true, soundWaveAssetPtr=&/Alien_Artifact_Loop_opt}}
]
properties {
 attenuationSettings=/Game/Sounds/Foley_Loot/Containers/SupplyDrops/Fort_SupplyDrop_Marker_Close_Athena_Att
 bOverrideAttenuation=true
 maxDistance=800
 soundClassObject=&Fort_Quests
 volumeMultiplier=1
 attenuationOverrides {falloffDistance=700, attenuationShapeExtents {x=100}}
}
