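# Sound cue for picking up an alien artifact in Athena, routed to the in-game non-critical UI sound class.
# A mixer layers a quieter reward chest-upgrade sting, the dedicated artifact pickup wave and a spotlight transmission sound.
# The transmission layer runs through a volume envelope with a quick fade-in and longer tail, then a fixed upward pitch shift.
# Concurrency allows one instance at a time, stopping the oldest.
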SoundCue /Game/Athena/Sounds/Quests/Athena_AlienArtifact_PickUP_Cue.Athena_AlienArtifact_PickUp_Cue
root=Mixer_0
nodes [
 {
  id=Enveloper_0
  children [WavePlayer_2]
  properties {
   pitchCurve {editorCurveData {keys [{time=0}]}}
   volumeCurve {editorCurveData {keys [<at 0.0039 0.0154>,<at 0.1937 1.0005>,<at 1.4999 1.0042>,<at 1.6714 0.5273>,<at 2.0779 0.1503>,<at 2.6259 0>]}}
  }
 }
 {id=Mixer_0, children [WavePlayer_1,WavePlayer_0,Modulator_0], properties {inputVolume [0.25,1,0.75]}}
 {id=Modulator_0, children [Enveloper_0], properties {pitchMax=1.1, pitchMin=1.1}}
 {id=WavePlayer_0, properties {soundWaveAssetPtr=&/Alein_Artifact_PickUp}}
 {id=WavePlayer_1, properties {soundWaveAssetPtr=/Game/Sounds/Fort_UI_Sounds/Rewards/Fort_RewardScreen_TeamScore_ChestLootUpgrade}}
 {
  id=WavePlayer_2
  properties {soundWaveAssetPtr=/Game/Athena/Sounds/Events/Mother/Ambient/LuxurySupreme/LuxurySupreme_Spotlight_Transmission_05}
 }
]
properties {
 bOverrideConcurrency=true
 maxDistance=2097152
 soundClassObject=&Fort_UI_InGame_NonCritical
 volumeMultiplier=1
 concurrencyOverrides {maxCount=1, resolutionRule=StopOldest}
}
