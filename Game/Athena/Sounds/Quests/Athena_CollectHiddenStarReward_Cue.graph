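# Sound cue for collecting a hidden star quest reward in Athena, judging by its name and folder.
# A single wave player reuses the reward screen's team score chest loot upgrade sound at reduced volume.
# It routes to the non-critical in-game UI sound class but overrides attenuation for distance falloff and stereo spread.
# A concurrency override allows one instance, stopping the oldest when a new one starts.
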SoundCue /Game/Athena/Sounds/Quests/Athena_CollectHiddenStarReward_Cue
root=WavePlayer_1
nodes [{id=WavePlayer_1, properties {soundWaveAssetPtr=/Game/Sounds/Fort_UI_Sounds/Rewards/Fort_RewardScreen_TeamScore_ChestLootUpgrade}}]
properties {
 bOverrideAttenuation=true
 bOverrideConcurrency=true
 maxDistance=1250
 soundClassObject=&Fort_UI_InGame_NonCritical
 volumeMultiplier=0.7
 attenuationOverrides {falloffDistance=850, stereoSpread=2300}
 concurrencyOverrides {maxCount=1, resolutionRule=StopOldest}
}
