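# Sound cue for collecting a floating lightning quest item in Athena, played in the SFX class.
# A quality-level root sends the first two levels to a mixer that layers the bolt pickup wave over a quieter storm lightning spawn wave.
# The other two levels play a single "opt" pickup wave instead of the mix.
# The cue overrides attenuation with its own shape extent and falloff distance.
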SoundCue /Game/Athena/Sounds/Quests/Athena_Quest_FloatingLightningCollect_Cue
root=QualityLevel_0
nodes [
 {id=Mixer_0, children [WavePlayer_0,WavePlayer_9], properties {inputVolume [1,0.25]}}
 {id=QualityLevel_0, children [Mixer_0,Mixer_0,WavePlayer_1,WavePlayer_1]}
 {id=WavePlayer_0, properties {soundWaveAssetPtr=&/Lightning_Bolt_PickUp_Athena_01}}
 {id=WavePlayer_1, properties {soundWaveAssetPtr=&/Lightning_Bolt_PickUp_Athena_opt}}
 {id=WavePlayer_9, properties {soundWaveAssetPtr=/Game/Sounds/Fort_GamePlay_Sounds/Storms/lightning_spawn_16_WAV_new}}
]
properties {
 bOverrideAttenuation=true
 maxDistance=4500
 soundClassObject=&SFX
 attenuationOverrides {falloffDistance=3000, attenuationShapeExtents {x=1500}}
}
