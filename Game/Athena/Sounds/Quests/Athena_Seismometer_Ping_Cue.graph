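# Sound cue for the seismometer ping in Athena quests, playing a tremor seismometer reading sound.
# A single wave player feeds the root directly, with no randomization or modulation, under a non-critical in-game UI sound class.
# Inline attenuation overrides spatialize it with short falloff and wide stereo spread.
# Concurrency overrides allow one instance at a time and block new plays while it sounds.
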SoundCue /Game/Athena/Sounds/Quests/Athena_Seismometer_Ping_Cue
root=WavePlayer_0
nodes [{id=WavePlayer_0, properties {soundWaveAssetPtr=&/Quest_Device_Tremor_Seismometer_Reading}}]
properties {
 bOverrideAttenuation=true
 bOverrideConcurrency=true
 maxDistance=1250
 soundClassObject=&Fort_UI_InGame_NonCritical
 volumeMultiplier=1.4
 attenuationOverrides {falloffDistance=850, stereoSpread=2300}
 concurrencyOverrides {maxCount=1, resolutionRule=PreventNew}
}
