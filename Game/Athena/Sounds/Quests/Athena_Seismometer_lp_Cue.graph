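# Looping sound cue for the tremor seismometer quest device, played in the Fort_Quests sound class.
# A quality-level node picks one of two looping wave players, each shared by two tiers and referencing a seismometer loop wave.
# The result feeds a root modulator that randomizes volume and pitch.
# Attenuation reuses a supply-drop ambient loop asset, with overrides present for falloff distance and shape extents.
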SoundCue /Game/Athena/Sounds/Quests/Athena_Seismometer_lp_Cue
root=Modulator_0
nodes [
 {id=Modulator_0, children [QualityLevel_0], properties {pitchMax=1, volumeMax=1.3, volumeMin=1}}
 {id=QualityLevel_0, children [WavePlayer_2,WavePlayer_2,WavePlayer_3,WavePlayer_3]}
 {id=WavePlayer_2, properties {bLooping=true, soundWaveAssetPtr=&/Quest_Device_Tremor_Seismometer_Loop}}
 {id=WavePlayer_3, properties {bLooping=true, soundWaveAssetPtr=&Quest_Device_Tremor_Seismometer_Loop}}
]
properties {
 attenuationSettings=/Game/Athena/Sounds/SupplyDrops/Fort_HiddenStar_Ambient_Loop_Att
 maxDistance=2500
 soundClassObject=&Fort_Quests
 volumeMultiplier=1.2
 attenuationOverrides {falloffDistance=700, attenuationShapeExtents {x=100}}
}
