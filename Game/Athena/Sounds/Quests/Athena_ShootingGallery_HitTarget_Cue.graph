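# Sound cue for hitting a target in the Athena shooting gallery quest, playing a gallery impact sound.
# A quality level node scales variety: five random impacts, three random impacts, or one fixed impact shared by the last two tiers.
# A modulator at the root wraps that selection.
# It uses gadget activation attenuation, the mission in-game sound class and rifle bullet impact concurrency.
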SoundCue /Game/Athena/Sounds/Quests/Athena_ShootingGallery_HitTarget_Cue
root=Modulator_4
nodes [
 {id=Modulator_4, children [QualityLevel_0]}
 {id=QualityLevel_0, children [Random_6, Random_7, WavePlayer_24, WavePlayer_24]}
 {id=Random_6, children [WavePlayer_4, WavePlayer_5, WavePlayer_14, WavePlayer_17, WavePlayer_18], properties {weights [1,1,1,1,1]}}
 {id=Random_7, children [WavePlayer_19, WavePlayer_20, WavePlayer_21], properties {weights [1,1,1]}}
 {id=WavePlayer_14, properties {soundWaveAssetPtr=&/FNBR_Gallery_Impact_03}}
 {id=WavePlayer_17, properties {soundWaveAssetPtr=&/FNBR_Gallery_Impact_04}}
 {id=WavePlayer_18, properties {soundWaveAssetPtr=&/FNBR_Gallery_Impact_05}}
 {id=WavePlayer_19, properties {soundWaveAssetPtr=&/FNBR_Gallery_Impact_01}}
 {id=WavePlayer_20, properties {soundWaveAssetPtr=&/FNBR_Gallery_Impact_02}}
 {id=WavePlayer_21, properties {soundWaveAssetPtr=&FNBR_Gallery_Impact_03}}
 {id=WavePlayer_24, properties {soundWaveAssetPtr=&FNBR_Gallery_Impact_01}}
 {id=WavePlayer_4, properties {soundWaveAssetPtr=&FNBR_Gallery_Impact_01}}
 {id=WavePlayer_5, properties {soundWaveAssetPtr=&FNBR_Gallery_Impact_02}}
]
properties {
 attenuationSettings=&Fort_GadgetActivate_Att
 maxDistance=4000
 soundClassObject=/Game/Sounds/FortSoundClassesAndModes/Fort_Mission_IngameSounds
 volumeMultiplier=1.3
 concurrencySet [&RifleBulletImpact_Conc]
}
