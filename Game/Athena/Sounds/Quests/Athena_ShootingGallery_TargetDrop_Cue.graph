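# Sound cue for the shooting gallery target drop in Athena quests, played with gadget-activate attenuation on the mission in-game sound class.
# A Quality Level node routes its first two slots to an equally weighted Random of three drop variations and its last two to a single fixed drop wave.
# A default Modulator at the root varies the result, and playback shares the rifle bullet impact concurrency set.
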SoundCue /Game/Athena/Sounds/Quests/Athena_ShootingGallery_TargetDrop_Cue
root=Modulator_0
nodes [
 {id=Modulator_0, children [QualityLevel_1]}
 {id=QualityLevel_1, children [Random_1,Random_1,WavePlayer_9,WavePlayer_9]}
 {id=Random_1, children [WavePlayer_6,WavePlayer_7,WavePlayer_8], properties {weights [1,1,1]}}
 {id=WavePlayer_6, properties {soundWaveAssetPtr=&/FNBR_Gallery_Drop_03}}
 {id=WavePlayer_7, properties {soundWaveAssetPtr=&/FNBR_Gallery_Drop_01}}
 {id=WavePlayer_8, properties {soundWaveAssetPtr=&/FNBR_Gallery_Drop_02}}
 {id=WavePlayer_9, properties {soundWaveAssetPtr=&FNBR_Gallery_Drop_01}}
]
properties {
 attenuationSettings=&Fort_GadgetActivate_Att
 maxDistance=4000
 soundClassObject=/Game/Sounds/FortSoundClassesAndModes/Fort_Mission_IngameSounds
 volumeMultiplier=1
 concurrencySet [&RifleBulletImpact_Conc]
}
