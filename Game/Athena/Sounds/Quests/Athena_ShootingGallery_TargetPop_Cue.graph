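# Sound cue for a shooting gallery target popping up, under the Athena quest sounds.
# A Modulator varies output from a quality switch: the first two levels randomly pick among three equally weighted "Up" waves, the last two play a single fixed wave.
# It borrows bullet-impact attenuation, sound class and rifle impact concurrency, so it spatializes and voice-limits like an impact sound.
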SoundCue /Game/Athena/Sounds/Quests/Athena_ShootingGallery_TargetPop_Cue
root=Modulator_0
nodes [
 {id=Modulator_0, children [QualityLevel_0]}
 {id=QualityLevel_0, children [Random_0,Random_0,WavePlayer_3,WavePlayer_3]}
 {id=Random_0, children [WavePlayer_0,WavePlayer_1,WavePlayer_2], properties {weights [1,1,1]}}
 {id=WavePlayer_0, properties {soundWaveAssetPtr=&/FNBR_Gallery_Up_03}}
 {id=WavePlayer_1, properties {soundWaveAssetPtr=&/FNBR_Gallery_Up_01}}
 {id=WavePlayer_2, properties {soundWaveAssetPtr=&/FNBR_Gallery_Up_02}}
 {id=WavePlayer_3, properties {soundWaveAssetPtr=&FNBR_Gallery_Up_01}}
]
properties {
 attenuationSettings=/Game/Sounds/Attenuation/Fort_Bullet_Impact_Att
 maxDistance=5300
 soundClassObject=&Fort_Impacts
 volumeMultiplier=1
 concurrencySet [&RifleBulletImpact_Conc]
}
