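# SoundCue for a clay pigeon duck launch sound, stored among Athena quest sounds.
# A single wave player plays one launch wave with no randomization or layering.
# It routes through the weapon foley sound class with gadget-activate attenuation and a slight volume boost.
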SoundCue /Game/Athena/Sounds/Quests/ClayPidgeon_DuckLaunch_Cue
root=WavePlayer_3
nodes [{id=WavePlayer_3, properties {soundWaveAssetPtr=&/Clay_Pigeon_DuckLaunch_01}}]
properties {
 attenuationSettings=&Fort_GadgetActivate_Att
 maxDistance=4000
 soundClassObject=&Fort_WeaponFoley
 volumeMultiplier=1.1
}
