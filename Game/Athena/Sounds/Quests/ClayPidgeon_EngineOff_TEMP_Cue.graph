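# Temporary Sound Cue for the clay pigeon engine-off sound, stored with Athena quest sounds.
# A single Wave Player plays the Clay_Pigeon_Stop_01 wave, pitched down by the cue's pitch multiplier.
# It uses the shared item-interact attenuation and routes through the Fort_Loot_Reward_PSM sound class.
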SoundCue /Game/Athena/Sounds/Quests/ClayPidgeon_EngineOff_TEMP_Cue
root=WavePlayer_1
nodes [{id=WavePlayer_1, properties {soundWaveAssetPtr=&/Clay_Pigeon_Stop_01}}]
properties {
 attenuationSettings=/Game/Sounds/Attenuation/Fort_Item_Interact_Att
 maxDistance=1850
 pitchMultiplier=0.7
 soundClassObject=&Fort_Loot_Reward_PSM
 volumeMultiplier=1
}
