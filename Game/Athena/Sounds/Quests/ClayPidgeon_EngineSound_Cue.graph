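# Sound cue for a clay pigeon engine in Athena quests, combining a start-up sound with a running loop.
# A mixer plays the one-shot start wave alongside the looping wave at slightly lower input volume.
# An enveloper ramps the loop up from low volume over a fraction of a second, easing it in under the start.
# It uses close-focus prop attenuation, the in-game mission sound class, and a concurrency override that prevents new instances at the limit.
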SoundCue /Game/Athena/Sounds/Quests/ClayPidgeon_EngineSound_Cue
root=Mixer_1
nodes [
 {
  id=Enveloper_1
  children [WavePlayer_5]
  properties {pitchCurve {editorCurveData {keys [{time=0}]}}, volumeCurve {editorCurveData {keys [<at 0 0.1>,<at 0.4 1>]}}}
 }
 {id=Mixer_1, children [WavePlayer_4,Enveloper_1], properties {inputVolume [1,0.8]}}
 {id=WavePlayer_4, properties {soundWaveAssetPtr=&/Clay_Pigeon_Start}}
 {id=WavePlayer_5, properties {bLooping=true, soundWaveAssetPtr=&/Clay_Pigeon_Loop}}
]
properties {
 attenuationSettings=/Game/Sounds/Attenuation/Fort_Prop_Close_Focus_HighFreqs_Att
 bOverrideAttenuation=true
 bOverrideConcurrency=true
 maxDistance=4000
 soundClassObject=/Game/Sounds/FortSoundClassesAndModes/Fort_Mission_IngameSounds
 concurrencyOverrides {resolutionRule=PreventNew}
}
