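# Temporary Sound Cue for a clay pigeon firing sound, filed under Athena quest sounds.
# A single Wave Player feeds the Clay_Pigeon_Fire_01 wave straight to the output, with no randomization or mixing.
# The cue lowers pitch and slightly boosts volume, and borrows the weapon foley sound class and weapon reload attenuation for its spatial falloff.
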SoundCue /Game/Athena/Sounds/Quests/ClayPidgeon_Shoot_TEMP_Cue
root=WavePlayer_2
nodes [{id=WavePlayer_2, properties {soundWaveAssetPtr=&/Clay_Pigeon_Fire_01}}]
properties {
 attenuationSettings=&Fort_Weapon_Reload_Att
 maxDistance=2250
 pitchMultiplier=0.8
 soundClassObject=&Fort_WeaponFoley
 volumeMultiplier=1.1
}
