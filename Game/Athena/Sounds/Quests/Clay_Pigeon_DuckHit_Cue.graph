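# Sound cue for the clay pigeon duck hit in Athena quests, mixing two layered sounds.
# A mixer plays a briefly delayed duck-hit wave alongside a louder, pitched-up soccer ball impact wave.
# It reuses gadget-activate attenuation and the third-person ranged weapon fire sound class, at reduced overall volume.
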SoundCue /Game/Athena/Sounds/Quests/Clay_Pigeon_DuckHit_Cue
root=Mixer_0
nodes [
 {id=Delay_0, children [WavePlayer_1], properties {delayMax=0.1, delayMin=0.1}}
 {id=Mixer_0, children [Delay_0,Modulator_0], properties {inputVolume [1,1.5]}}
 {id=Modulator_0, children [WavePlayer_2], properties {pitchMax=1.5, pitchMin=1.5}}
 {id=WavePlayer_1, properties {soundWaveAssetPtr=&/Clay_Pigeon_DuckHit_01}}
 {id=WavePlayer_2, properties {soundWaveAssetPtr=/Game/Sounds/Fort_Impact_Sounds/Soccer/SoccerBall_Hit_01}}
]
properties {
 attenuationSettings=&Fort_GadgetActivate_Att
 maxDistance=4000
 soundClassObject=&Fort_WeaponRanged_Fire_3P_PSM
 volumeMultiplier=0.5
}
