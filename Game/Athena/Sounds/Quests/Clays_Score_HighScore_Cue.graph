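# Sound cue for a high-score sound among the Athena quest sounds, apparently for a clays shooting score.
# A single wave player feeds the Clays_Score_HighScore wave straight to the output, with no randomization or mixing.
# It plays at boosted volume in the mission in-game sound class, spatialized through the shared gadget-activate attenuation asset.
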SoundCue /Game/Athena/Sounds/Quests/Clays_Score_HighScore_Cue
root=WavePlayer_0
nodes [{id=WavePlayer_0, properties {soundWaveAssetPtr=&/Clays_Score_HighScore}}]
properties {
 attenuationSettings=&Fort_GadgetActivate_Att
 maxDistance=4000
 soundClassObject=/Game/Sounds/FortSoundClassesAndModes/Fort_Mission_IngameSounds
 volumeMultiplier=2
}
