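# Looping sound cue for dance floor music, filed under Athena quest sounds.
# A quality-level node picks between two looping wave players: the standard music loop for the first two quality tiers and an OPT variant, presumably optimized, for the last two.
# It is spatialized through a dance floor attenuation asset and routed to the quest music sound class.
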SoundCue /Game/Athena/Sounds/Quests/DanceFloor_Loop_Cue
root=QualityLevel_0
nodes [
 {id=QualityLevel_0, children [WavePlayer_0,WavePlayer_0,WavePlayer_1,WavePlayer_1]}
 {id=WavePlayer_0, properties {bLooping=true, soundWaveAssetPtr=&/Music_DanceFloor_Loop.music_dancefloor_loop}}
 {id=WavePlayer_1, properties {bLooping=true, soundWaveAssetPtr=&/music_dancefloor_loop_OPT}}
]
properties {
 attenuationSettings=/Game/Sounds/Attenuation/Fort_DanceFloorMusic_Athena_Att
 maxDistance=1280
 soundClassObject=/Game/Sounds/FortSoundClassesAndModes/Fort_Music_Quest_PSM
}
