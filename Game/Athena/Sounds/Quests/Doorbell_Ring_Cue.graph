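# Sound cue for a doorbell ring, filed under Athena quest sounds, that builds a two-tone ding-dong from a church bell recording.
# A mixer layers two players of the same wave, both pitched well up, with the second slightly lower and delayed half a second.
# The mix feeds an attenuation node at the root.
# The cue uses the ambient tile-mix sound class and overrides concurrency to block new instances beyond a small limit.
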SoundCue /Game/Athena/Sounds/Quests/Doorbell_Ring_Cue
root=Attenuation_2
nodes [
 {id=Attenuation_2, children [Mixer_0]}
 {id=Delay_0, children [Modulator_1], properties {delayMax=0.5, delayMin=0.5}}
 {id=Mixer_0, children [Modulator_0,Delay_0], properties {inputVolume [1,1]}}
 {id=Modulator_0, children [WavePlayer_0], properties {pitchMax=2, pitchMin=2, volumeMax=1.1}}
 {id=Modulator_1, children [WavePlayer_7], properties {pitchMax=1.7, pitchMin=1.7, volumeMax=1.1}}
 {id=WavePlayer_0, properties {soundWaveAssetPtr=/Game/Sounds/Fort_Prop_Sounds/Exterior/church_bell_ring_01}}
 {id=WavePlayer_7, properties {soundWaveAssetPtr=&church_bell_ring_01}}
]
properties {
 bOverrideAttenuation=true
 bOverrideConcurrency=true
 maxDistance=2097152
 soundClassObject=/Game/Sounds/FortSoundClassesAndModes/Fort_Ambient_TileMix
 volumeMultiplier=0.375
 concurrencyOverrides {maxCount=5, resolutionRule=PreventNew}
}
