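# Looping ambient sound cue for the Fortnite letters quest, judging by its name, reusing supply drop target loop audio as a positional hint.
# A quality level node picks between two looping waves: the standard loop for the first two levels, the "opt" variant for the last two.
# Playback uses the quest interactables hint attenuation and the UI sound class, with a slight volume boost.
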SoundCue /Game/Athena/Sounds/Quests/FortniteLetters_AmbientLoop_Cue
root=QualityLevel_0
nodes [
 {id=QualityLevel_0, children [WavePlayer_0,WavePlayer_0,WavePlayer_2,WavePlayer_2]}
 {id=WavePlayer_0, properties {bLooping=true, soundWaveAssetPtr=/Game/Sounds/Foley_Loot/Containers/SupplyDrops/sfx_supplydrop_target_loop}}
 {id=WavePlayer_2, properties {bLooping=true, soundWaveAssetPtr=/Game/Sounds/Foley_Loot/Containers/SupplyDrops/sfx_supplydrop_target_loop_opt}}
]
properties {
 attenuationSettings=/Game/Sounds/Attenuation/Fort_QuestInteractables_Hint
 maxDistance=1300
 soundClassObject=&Fort_UI
 volumeMultiplier=1.1
}
