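# Sound cue for collecting Fortnite letters, a quest pickup, built from front-end UI sounds.
# A mixer layers a row-select sound and two slot-item sounds, with the second slot-item sound delayed briefly to trail the others.
# The other delay node keeps its default timing.
# The cue uses the UI sound class with quest-interactable hint attenuation and a slight volume boost.
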SoundCue /Game/Athena/Sounds/Quests/FortniteLetters_Collect_Cue
root=Mixer_0
nodes [
 {id=Delay_1, children [WavePlayer_3], properties {delayMax=0.4, delayMin=0.4}}
 {id=Delay_2, children [WavePlayer_2]}
 {id=Mixer_0, children [WavePlayer_0,Delay_2,Delay_1], properties {inputVolume [1,1,1]}}
 {id=WavePlayer_0, properties {soundWaveAssetPtr=/Game/Sounds/Fort_UI_Sounds/FrontEnd/Fort_UI_CB_SelectRow_01}}
 {id=WavePlayer_2, properties {soundWaveAssetPtr=/Game/Sounds/Fort_UI_Sounds/FrontEnd/Fort_UI_CB_SlotItem_01}}
 {id=WavePlayer_3, properties {soundWaveAssetPtr=/Game/Sounds/Fort_UI_Sounds/FrontEnd/Fort_UI_CB_SlotItem_02}}
]
properties {
 attenuationSettings=/Game/Sounds/Attenuation/Fort_QuestInteractables_Hint
 maxDistance=1300
 soundClassObject=&Fort_UI
 volumeMultiplier=1.1
}
