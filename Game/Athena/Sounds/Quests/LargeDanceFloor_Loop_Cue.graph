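# Looping music Sound Cue for the large Athena dance floor, routed to the quest music sound class.
# A Mixer layers a near branch with a slightly quieter distant branch, each with its own attenuation asset.
# In each branch, a quality-level switch picks a looping standard wave for the first two levels and an _OPT variant for the last two.
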SoundCue /Game/Athena/Sounds/Quests/LargeDanceFloor_Loop_Cue
root=Mixer_1
nodes [
 {id=Attenuation_2, children [QualityLevel_0], properties {attenuationSettings=/Game/Sounds/Attenuation/Fort_DanceFloorMusic_Large_Athena_Att}}
 {
  id=Attenuation_3
  children [QualityLevel_1]
  properties {attenuationSettings=/Game/Sounds/Attenuation/Fort_DanceFloorMusic_Large_Distant_Athena_Att}
 }
 {id=Mixer_1, children [Attenuation_2, Attenuation_3], properties {inputVolume [1,0.75]}}
 {id=QualityLevel_0, children [WavePlayer_0,WavePlayer_0,WavePlayer_1,WavePlayer_1]}
 {id=QualityLevel_1, children [WavePlayer_2,WavePlayer_2,WavePlayer_3,WavePlayer_3]}
 {id=WavePlayer_0, properties {bLooping=true, soundWaveAssetPtr=&/largedancefloor_music_01}}
 {id=WavePlayer_1, properties {bLooping=true, soundWaveAssetPtr=&/largedancefloor_music_01_OPT}}
 {id=WavePlayer_2, properties {bLooping=true, soundWaveAssetPtr=&largedancefloor_music_01}}
 {id=WavePlayer_3, properties {bLooping=true, soundWaveAssetPtr=&largedancefloor_music_01_OPT}}
]
properties {maxDistance=3536, soundClassObject=/Game/Sounds/FortSoundClassesAndModes/Fort_Music_Quest_PSM}
