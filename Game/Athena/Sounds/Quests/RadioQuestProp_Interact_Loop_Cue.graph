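# Sound cue for interacting with a radio quest prop, judging by its name.
# A mixer layers a one-shot chest-upgrade reward sound at full level over a much quieter looping glitchy radio comms bed.
# It routes through the non-critical in-game UI sound class but overrides attenuation for distance falloff.
# A concurrency override allows one instance, stopping the oldest.
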SoundCue /Game/Athena/Sounds/Quests/RadioQuestProp_Interact_Loop_Cue
root=Mixer_0
nodes [
 {id=Mixer_0, children [WavePlayer_1,WavePlayer_0], properties {inputVolume [1,0.2]}}
 {id=WavePlayer_0, properties {bLooping=true, soundWaveAssetPtr=/Game/Sounds/Helios/POI/SocietyEstate/Spots/Spot_Radio_Comms_Glitchy_Loop}}
 {id=WavePlayer_1, properties {soundWaveAssetPtr=/Game/Sounds/Fort_UI_Sounds/Rewards/Fort_RewardScreen_TeamScore_ChestLootUpgrade}}
]
properties {
 bOverrideAttenuation=true
 bOverrideConcurrency=true
 maxDistance=1250
 soundClassObject=&Fort_UI_InGame_NonCritical
 volumeMultiplier=0.7
 attenuationOverrides {falloffDistance=850, stereoSpread=2300}
 concurrencyOverrides {maxCount=1, resolutionRule=StopOldest}
}
