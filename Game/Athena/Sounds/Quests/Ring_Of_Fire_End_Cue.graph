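# Sound cue for the end sound of the Ring of Fire quest in Athena.
# A single wave player plays the Ring_Of_Fire_End wave at boosted volume through the Fort_Quests sound class.
# It uses the backpack hint attenuation asset.
# A concurrency override limits it to one instance per owner and blocks new plays while one is active.
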SoundCue /Game/Athena/Sounds/Quests/Ring_Of_Fire_End_Cue
root=WavePlayer_0
nodes [{id=WavePlayer_0, properties {soundWaveAssetPtr=&/Ring_Of_Fire_End}}]
properties {
 attenuationSettings=/Game/Sounds/Attenuation/Fort_BackPack_Hint_Att
 bOverrideConcurrency=true
 maxDistance=1450
 soundClassObject=&Fort_Quests
 volumeMultiplier=2.24
 concurrencyOverrides {bLimitToOwner=true, maxCount=1, resolutionRule=PreventNew}
}
