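# Looping sound cue for a Ring of Fire quest element in Athena, routed to the quests sound class.
# A quality-level node selects between two looping wave players: the full Ring_Of_Fire_Loop for the first two tiers and the OPT variant for the other two.
# It uses the quest-interactables hint attenuation with a boosted volume.
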SoundCue /Game/Athena/Sounds/Quests/Ring_Of_Fire_Loop_Cue
root=QualityLevel_0
nodes [
 {id=QualityLevel_0, children [WavePlayer_0,WavePlayer_0,WavePlayer_1,WavePlayer_1]}
 {id=WavePlayer_0, properties {bLooping=true, soundWaveAssetPtr=&/Ring_Of_Fire_Loop}}
 {id=WavePlayer_1, properties {bLooping=true, soundWaveAssetPtr=&/Ring_Of_Fire_Loop_OPT}}
]
properties {
 attenuationSettings=/Game/Sounds/Attenuation/Fort_QuestInteractables_Hint
 maxDistance=1300
 soundClassObject=&Fort_Quests
 volumeMultiplier=2.24
}
