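# Sound cue for collecting a bronze score coin, filed under Athena quest sounds.
# A single wave player plays the bronze coin pickup wave at reduced volume, with no randomization or other processing nodes.
# The cue uses the Fort_UI sound class but overrides attenuation with a short falloff distance, so it is set up for distance-based fade rather than flat UI playback.
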SoundCue /Game/Athena/Sounds/Quests/ScoreCoin_Collect_Bronze_Cue
root=WavePlayer_1
nodes [{id=WavePlayer_1, properties {soundWaveAssetPtr=&/ScoreCoin_Pickup_Bronze_01}}]
properties {
 bOverrideAttenuation=true
 maxDistance=727
 soundClassObject=&Fort_UI
 volumeMultiplier=0.6
 attenuationOverrides {falloffDistance=512, attenuationShapeExtents {x=215}}
}
