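# Sound cue for the spin sound of a score coin, filed under Athena quest sounds.
# A single wave player feeds the ScoreCoin_Spin_04 wave straight to the output, with no randomization or mixing, at slightly boosted volume in the Fort_UI sound class.
# Overridden attenuation gives it a short positional falloff with stereo spread, and overridden concurrency caps simultaneous instances at a few, stopping the oldest first.
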SoundCue /Game/Athena/Sounds/Quests/ScoreCoin_Spin_Cue
root=WavePlayer_0
nodes [{id=WavePlayer_0, properties {soundWaveAssetPtr=&/ScoreCoin_Spin_04}}]
properties {
 bOverrideAttenuation=true
 bOverrideConcurrency=true
 maxDistance=1536
 soundClassObject=&Fort_UI
 volumeMultiplier=1.1
 attenuationOverrides {bEnableFocusInterpolation=true, falloffDistance=1536, stereoSpread=512, attenuationShapeExtents {x=0}}
 concurrencyOverrides {maxCount=3, resolutionRule=StopOldest}
}
