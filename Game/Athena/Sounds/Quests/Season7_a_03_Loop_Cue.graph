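# Season 7 quest sound cue that plays a looping ambience in the quest sound class.
# The first two quality levels mix a one-shot season7_a_04 wave with looping season7_a_03, which an Enveloper fades in over about a second.
# The other two levels substitute a single optimized supply drop target loop.
# Attenuation borrows the close supply drop marker asset, though the cue enables its own short-range override.
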SoundCue /Game/Athena/Sounds/Quests/Season7_a_03_Loop_Cue
root=QualityLevel_0
nodes [
 {
  id=Enveloper_0
  children [WavePlayer_2]
  properties {pitchCurve {editorCurveData {keys [{time=0}]}}, volumeCurve {editorCurveData {keys [<at 0 0.0354>,<at 0.4502 0.176>,<at 1.1773 1.0532>]}}}
 }
 {id=Mixer_0, children [WavePlayer_3,Enveloper_0], properties {inputVolume [1,1]}}
 {id=QualityLevel_0, children [Mixer_0,Mixer_0,WavePlayer_1,WavePlayer_1]}
 {id=WavePlayer_1, properties {bLooping=true, soundWaveAssetPtr=/Game/Sounds/Foley_Loot/Containers/SupplyDrops/sfx_supplydrop_target_loop_opt}}
 {id=WavePlayer_2, properties {bLooping=true, soundWaveAssetPtr=&/season7_a_03}}
 {id=WavePlayer_3, properties {soundWaveAssetPtr=&/season7_a_04}}
]
properties {
 attenuationSettings=/Game/Sounds/Foley_Loot/Containers/SupplyDrops/Fort_SupplyDrop_Marker_Close_Athena_Att
 bOverrideAttenuation=true
 maxDistance=800
 soundClassObject=&Fort_Quests
 volumeMultiplier=0.672
 attenuationOverrides {falloffDistance=700, attenuationShapeExtents {x=100}}
}
